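# Niagara system for the Falling Star vehicle boost cosmetic, drawing a star with a trailing ribbon and sprite particles at an offset boost position.
# The Star emitter keeps looping mesh and sprite particles alive with curve-driven rotation.
# Sprite and ribbon emitters shoot backward, inheriting source movement that fades over lifetime.
# User painted flag and three colors switch each layer from default colors, also adjusting ribbon and star material parameters.
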NiagaraSystem /VehicleCosmetics/Boosts/Boost_Fallingstar/FX/NS_Boost_FallingStar
dependencies {
 usedDataInterfaces [&NiagaraDataInterfaceVectorCurve, &NiagaraDataInterfaceCurve, &NiagaraDataInterfaceSkeletalMesh]
 usedDynamicInputs []
 usedModules [
  &self:SystemSpawnScript.SystemScriptSource.SystemScriptGraph.NiagaraNodeAssignment_0.SetVariables_#1
  &SystemState
 ]
}
emitters {
 NE_VC_Boost_Sprite_Parent {
  emitterUpdateScript [
   &EmitterState()
   &VC_FilteredSocketCount()
   &SpawnRate(SpawnRate=&RandomRangeFloat(Minimum=8, Maximum=14), Use Spawn Probability=false)
  ]
  particleSpawnScript [
   &VC_PaintedVariantsOverride() {enabled=false}
   &InitializeParticle(
    Color Mode=::NewEnumerator0
    Lifetime=&MakeCustomFloatFromBool(Bool=$User.bIsFrontendPreview, True Float=&RandomRangeFloat(Minimum=0.5, Maximum=1), False Float=1)
    Position Offset=&AddVectorToPosition(Position=$System.FinalPosition, Vector=<-47 0 -3.4091>)
    Position Offset Coordinate Space=::Local
    Sprite Size Max=<0.6 30>
    Sprite Size Min=<0.5 10>
    Sprite Size Mode=::NewEnumerator2
    UsePositionOffset=true
    Write Material Random=true
   )
   &PartitionParticles(Number Of Partitions=$Emitter.PortCount)
   &SkeletalMeshLocation(
    Bone / Socket Index=$Particles.PartitionParticles.Partition
    Bone / Socket Sampling Mode=::NewEnumerator5
    MaskSpawnGroup=false
    Mesh Sampling Type=::NewEnumerator6
    Skeletal Mesh=$System.SkeletalMesh
   ) {enabled=false}
   &AddVelocity(Velocity=&RandomRangeVector(Minimum=<-1000 0 0>, Maximum=<-1500 0 0>))
   &Color(Color=&MakeCustomLinearColorFromBool(Bool=$User.bIsPainted, True Color=$User.PaintedColor3, False Color=<rgba 1.075 0.375 1.25 1>))
   &ShapeLocation(MaskSpawnGroup=false, Override Seed=false, Sphere Radius=25, UseOwnerScale=true)
  ]
  particleUpdateScript [
   &ParticleState()
   &ScaleSpriteSize(Scale Sprite Size Mode=::NewEnumerator0, Uniform Scale Factor=2)
   &InheritSourceMovement(
    Applied Position Scale=&VectorFromFloat(Value=&FloatFromCurve(FloatCurve=DataInterface() {propertyValues {Curve {keys [
          {interpMode=RCIM_Cubic, tangentMode=RCTM_Break, time=0, value=1}
          {arriveTangent=-2.9025, interpMode=RCIM_Cubic, tangentMode=RCTM_Break, time=1, value=0}
         ]}}}))
    ApplyForce=false
    ApplyPosition=true
    ApplyPositionFalloff=false
   )
   &SolveForcesAndVelocity(Clamp Velocity=false, Limit Acceleration=false)
   &DynamicMaterialParameters(
    Index 0 Param 3=2
    Index 0 Param 4=3
    Module.Index0Param0Write=true
    Module.Index0Param1Write=true
    Module.Index0Param2Write=true
    Module.Index0Param3Write=true
   )
   &ScaleColor(
    Scale Alpha=&FloatFromCurve(FloatCurve=DataInterface() {propertyValues {Curve {keys [<at 0 1>,<at 0.1 1>,<at 0.2 1>,<at 0.25 0>]}}})
    ScaleA=true
    ScaleRGB=true
   )
  ]
  rendererClasses [&NiagaraSpriteRendererProperties]
  renderers [{rendererClass=&NiagaraSpriteRendererProperties}]
 }
 NE_VC_SingleBoost_Ribbon_Parent {
  emitterUpdateScript [
   &EmitterState()
   &VC_FilteredSocketCount()
   &VC_RateQualityHandler(Epic Spawn Rate=30, High Spawn Rate=30, Low Spawn Rate=30, Use Custom Rate?=true)
   &VC_SpawnRibbon_v1(SpawnRate=$Emitter.Spawn Rate)
  ]
  particleSpawnScript [
   &VC_PaintedVariantsOverride(Pink=<rgba 0.5938 0.1175 0.4206 1>, Purple=<rgba 0.4473 0.0002 0.849 1>, SkyBlue=<rgba 0.1997 0.7045 0.9583 1>) {enabled=false}
   &VC_InitializeRibbon_v1(
    Custom Offset=true
    Lifetime=0.85
    Local Vector Offset=&AddVectorToPosition(Position=$System.FinalPosition, Vector=<-40 0 -5>)
    Offset Amount=1
    Ribbon Width=65
   )
   &AddVelocity(Velocity=<-1000 0 0>, Velocity Speed Scale=2) {module=AddVelocity001}
  ]
  particleUpdateScript [
   &ParticleState()
   &InheritSourceMovement(
    Applied Position Scale=&VectorFromFloat(Value=&FloatFromCurve(FloatCurve=DataInterface() {propertyValues {Curve {keys [
          {interpMode=RCIM_Cubic, tangentMode=RCTM_Break, time=0, value=1}
          {arriveTangent=-2.9025, interpMode=RCIM_Cubic, tangentMode=RCTM_Break, time=1, value=0}
         ]}}}))
    ApplyForce=false
    ApplyPosition=true
    ApplyPositionFalloff=false
   )
   &SolveForcesAndVelocity(Clamp Velocity=false, Limit Acceleration=false)
   &VC_UpdateRibbon_v1()
   &CameraOffset(Camera Offset Amount=100) {enabled=false}
   &DynamicMaterialParameters(
    Index 0 Param 1=&MakeCustomFloatFromBool(Bool=$User.bIsPainted, True Float=2, False Float=1)
    Index 0 Param 2=&MakeCustomFloatFromBool(Bool=$User.bIsPainted, True Float=$User.intensity_trail, False Float=5)
    Index 0 Param 3=&MakeCustomFloatFromBool(Bool=$User.bIsPainted, True Float=$User.Lines_ribbon, False Float=5)
    Module.Index0Param0Write=true
    Module.Index0Param1Write=true
    Module.Index0Param2Write=true
    Module.Index0Param3Write=true
   )
   &ScaleRibbonWidth(Ribbon Width Scale=&FloatFromCurve(
     FloatCurve=DataInterface() {propertyValues {Curve {keys [
         {interpMode=RCIM_Cubic, leaveTangent=-2.3378, tangentMode=RCTM_Break, time=0, value=1}
         {arriveTangent=-0.6884, interpMode=RCIM_Cubic, leaveTangent=-0.8888, tangentMode=RCTM_Break, time=0.25, value=0.65}
         {arriveTangent=-0.3446, interpMode=RCIM_Cubic, tangentMode=RCTM_Break, time=0.6, value=0.3}
        ]}}}
     Scale Curve=1.1
    ))
   &ScaleColor(
    Scale RGB=&MakeCustomVectorFromBool(
     Bool=$User.bIsPainted
     True Vector=&MakeVectorFromLinearColorRGB(Color=$User.PaintedColor2)
     False Vector=&VectorFromCurve(
      VectorCurve=DataInterface() {propertyValues {XCurve {keys [<at 0 0.05>,<at 0.5 0.5>]}, YCurve {keys [<at 0 0.5>,<at 0.5 0.1>]}, ZCurve {keys [<at 0 1>,<at 0.5 0.6>]}}}
      Scale Curve=&VectorFromFloat(Value=&FloatFromCurve(FloatCurve=DataInterface() {propertyValues {Curve {keys [<at 0 1>,<at 1 0>]}}}, Scale Curve=3))
     )
    )
    ScaleA=true
    ScaleRGB=true
   ) {enabled=false}
   &ScaleColor(
    Scale RGB=&MakeCustomVectorFromBool(
     Bool=$User.bIsPainted
     True Vector=&MakeVectorFromLinearColorRGB(Color=$User.PaintedColor2)
     False Vector=&VectorFromCurve(
      VectorCurve=DataInterface() {propertyValues {
        XCurve {keys [
          {interpMode=RCIM_Cubic, leaveTangent=0.1069, tangentMode=RCTM_Break, time=0, value=0.2}
          {arriveTangent=0.1189, interpMode=RCIM_Cubic, tangentMode=RCTM_Break, time=0.2, value=0.243}
          {arriveTangent=0.0029, interpMode=RCIM_Cubic, leaveTangent=0.0029, tangentMode=RCTM_Break, time=0.3, value=0.243}
          {interpMode=RCIM_Cubic, tangentMode=RCTM_Break, time=0.35, value=0.447}
         ]}
        YCurve {keys [
          {interpMode=RCIM_Cubic, leaveTangent=-1.3735, tangentMode=RCTM_Break, time=0, value=0.704}
          {arriveTangent=-1.5466, interpMode=RCIM_Cubic, tangentMode=RCTM_Break, time=0.2, value=0.234}
          {arriveTangent=0.0229, interpMode=RCIM_Cubic, leaveTangent=0.047, tangentMode=RCTM_Break, time=0.3, value=0.234}
          {arriveTangent=-0.2622, interpMode=RCIM_Cubic, tangentMode=RCTM_Break, time=0.6, value=0}
         ]}
        ZCurve {keys [
          {interpMode=RCIM_Cubic, leaveTangent=-0.4505, tangentMode=RCTM_Break, time=0, value=0.958}
          {arriveTangent=-0.4035, interpMode=RCIM_Cubic, tangentMode=RCTM_Break, time=0.2, value=0.849}
          {arriveTangent=0.0014, interpMode=RCIM_Cubic, leaveTangent=0.0014, tangentMode=RCTM_Break, time=0.3, value=0.849}
          {interpMode=RCIM_Cubic, tangentMode=RCTM_Break, time=0.4, value=0.9075}
         ]}
       }}
      Scale Curve=&VectorFromFloat(Value=&FloatFromCurve(
        FloatCurve=DataInterface() {propertyValues {Curve {keys [
            {interpMode=RCIM_Cubic, tangentMode=RCTM_Break, time=0, value=1}
            {arriveTangent=-1.5669, interpMode=RCIM_Cubic, tangentMode=RCTM_Break, time=1, value=0}
           ]}}}
        Scale Curve=2
       ))
     )
    )
    ScaleA=true
    ScaleRGB=true
   ) {enabled=false, module=ScaleColor001}
   &ScaleColor(
    Scale Alpha=&MakeCustomFloatFromBool(
     Bool=$User.bIsPainted
     True Float=&FloatFromCurve(FloatCurve=DataInterface() {propertyValues {Curve {keys [<at 0 1>,<at 1 0>]}}}, Scale Curve=1.5)
     False Float=1
    )
    Scale RGB=&MakeCustomVectorFromBool(
     Bool=$User.bIsPainted
     True Vector=&MakeVectorFromLinearColorRGB(Color=$User.PaintedColor2)
     False Vector=&VectorFromCurve(
      VectorCurve=DataInterface() {propertyValues {
        XCurve {keys [
          {interpMode=RCIM_Cubic, tangentMode=RCTM_Break, time=0, value=0.25}
          {arriveTangent=0.5155, interpMode=RCIM_Cubic, tangentMode=RCTM_Break, time=0.5, value=0.5}
         ]}
        YCurve {keys [
          {interpMode=RCIM_Cubic, leaveTangent=-1.5136, tangentMode=RCTM_Break, time=0, value=0.704}
          {interpMode=RCIM_Cubic, tangentMode=RCTM_Break, time=0.5, value=0.1}
         ]}
        ZCurve {keys [
          {interpMode=RCIM_Cubic, leaveTangent=-0.7415, tangentMode=RCTM_Break, time=0, value=1.5}
          {interpMode=RCIM_Cubic, tangentMode=RCTM_Break, time=0.5, value=0.5}
         ]}
       }}
      Scale Curve=&VectorFromFloat(Value=&FloatFromCurve(
        FloatCurve=DataInterface() {propertyValues {Curve {keys [
            {interpMode=RCIM_Cubic, leaveTangent=0.196, tangentMode=RCTM_Break, time=0, value=1}
            {interpMode=RCIM_Cubic, leaveTangent=1.2601, tangentMode=RCTM_Break, time=1, value=0}
           ]}}}
        Scale Curve=1.2
       ))
     )
    )
    ScaleA=true
    ScaleRGB=true
   ) {module=ScaleColor002}
  ]
  rendererClasses [&NiagaraRibbonRendererProperties]
  renderers [{rendererClass=&NiagaraRibbonRendererProperties}]
 }
 Star {
  emitterUpdateScript [
   &EmitterState(
    Inactive Response=::NewEnumerator1
    Life Cycle Mode=::NewEnumerator1
    Loop Behavior=::NewEnumerator1
    Loop Duration Mode=::NewEnumerator1
    UseLoopDelay=false
   )
   &SpawnBurst_Instantaneous(Use Spawn Probability=false, UseLoopCountLimit=false)
   &VC_FilteredSocketCount()
   &SpawnRate(SpawnRate=50, Use Spawn Probability=false) {enabled=false}
   /SetVariables(Emitter.IsnotPainted:NiagaraBool=&Boolean_NotOperation(Bool=$User.bIsPainted)) {isSetParameters=true}
  ]
  particleSpawnScript [
   &VC_PaintedVariantsOverride() {enabled=false}
   &InitializeParticle(
    Color Mode=::NewEnumerator0
    Lifetime=2.5
    Mesh Scale Mode=::NewEnumerator3
    Mesh Uniform Scale=0.7
    Position Offset=&AddVectorToPosition(Position=Linked() {linkedVariable {name=System.FinalPosition, type=Vector3f}}, Vector=<-47 0 -3.4091>)
    UsePositionOffset=true
    Write Material Random=true
   )
   &PartitionParticles(Number Of Partitions=$Emitter.PortCount) {enabled=false}
   &SkeletalMeshLocation(
    Bone / Socket Index=$Particles.PartitionParticles.Partition
    Bone / Socket Sampling Mode=::NewEnumerator5
    MaskSpawnGroup=false
    Mesh Sampling Type=::NewEnumerator6
    Sampled Position Offset=&AddVectorToPosition(
     Position=<0 0 0>
     Vector=&AddVectorToPosition(Position=$User.SingleBoostPos, Vector=$User.DTCarOffset)
    )
    Skeletal Mesh=$System.SkeletalMesh
   ) {enabled=false}
   &AddVelocity(Velocity=<-1500 0 0>) {enabled=false}
   &Color(Color=&MakeCustomLinearColorFromBool(Bool=$User.bIsPainted, True Color=$User.PaintedColor, False Color=<rgba 0.05 0.3 1 1>))
   &InitialMeshOrientation(
    Mesh Orientation Mode=::NewEnumerator2
    Orientation Vector=&RandomRangeVector(Minimum=<-1 0 0>, Maximum=<1 0 0>)
    Use Rotation Vector=false
   ) {enabled=false}
  ]
  particleUpdateScript [
   &ParticleState(Kill Particles When Lifetime Has Elapsed=false, Loop Particles Lifetime=true)
   &UpdateMeshOrientation(Rotation Vector=&VectorFromCurve(
     VectorCurve=DataInterface() {propertyValues {
       XCurve {keys [
         <at 0 0 RCIM_Cubic>
         {arriveTangent=0.0249, interpMode=RCIM_Cubic, leaveTangent=0.0249, tangentMode=RCTM_User, time=0.25, value=1}
         {arriveTangent=0.0245, interpMode=RCIM_Cubic, leaveTangent=0.0245, tangentMode=RCTM_User, time=0.75, value=-1}
         <at 1 0 RCIM_Cubic>
        ]}
       YCurve {keys [<at 0 1>,<at 1 0>]}
       ZCurve {keys [<at 0 1>,<at 1 0>]}
      }}
     Scale Curve=<1 0 0>
    ))
   &ScaleRibbonWidth(Ribbon Width Scale=3) {enabled=false}
   &MeshRotationForce(Rotation=&VectorFromCurve(
     VectorCurve=DataInterface() {propertyValues {
       XCurve {keys [
         {arriveTangent=0.0377, interpMode=RCIM_Cubic, leaveTangent=0.0377, tangentMode=RCTM_User, time=0, value=0}
         {arriveTangent=0.1488, interpMode=RCIM_Cubic, leaveTangent=0.1488, tangentMode=RCTM_User, time=0.25, value=1}
         {arriveTangent=0.0316, interpMode=RCIM_Cubic, leaveTangent=0.0316, tangentMode=RCTM_User, time=0.5, value=-1}
         {arriveTangent=-0.0224, interpMode=RCIM_Cubic, leaveTangent=-0.0224, tangentMode=RCTM_User, time=0.75, value=1}
         {arriveTangent=0.0805, interpMode=RCIM_Cubic, leaveTangent=0.0805, tangentMode=RCTM_User, time=1, value=0}
        ]}
       YCurve {keys [<at 0 1>,<at 1 0>]}
       ZCurve {keys [<at 0 1>,<at 1 0>]}
      }}
     Scale Curve=<1 0 0>
    )) {enabled=false}
   &SolveForcesAndVelocity(Clamp Velocity=false, Limit Acceleration=false)
   &ScaleMeshSize(Scale Factor=&VectorFromFloat(Value=3))
   &DynamicMaterialParameters(
    Index 0 Param 1=&MakeCustomFloatFromBool(Bool=$User.bIsPainted, True Float=1, False Float=0)
    Index 0 Param 2=&MakeCustomFloatFromBool(Bool=$User.bIsPainted, True Float=$User.OpacityMiniStars, False Float=0)
    Index 0 Param 3=$User.LinesinStar
    Module.Index0Float0Write=true
    Module.Index0Float1Write=true
    Module.Index0Float2Write=true
    Module.Index0Float3Write=true
   )
   &CameraOffset(Camera Offset Amount=50)
  ]
  rendererClasses [&NiagaraMeshRendererProperties, &NiagaraSpriteRendererProperties]
  renderers [
   {rendererClass=&NiagaraMeshRendererProperties, rendererIndex=0}
   {rendererClass=&NiagaraMeshRendererProperties, rendererIndex=1}
   {rendererClass=&NiagaraSpriteRendererProperties, rendererIndex=2}
  ]
 }
}
summary {
 systemName=NS_Boost_FallingStar
 emitters [
  {emitterName=Star, rendererClasses [&NiagaraMeshRendererProperties, &NiagaraSpriteRendererProperties]}
  {emitterName=NE_VC_Boost_Sprite_Parent, rendererClasses [&NiagaraSpriteRendererProperties]}
  {emitterName=NE_VC_SingleBoost_Ribbon_Parent, rendererClasses [&NiagaraRibbonRendererProperties]}
 ]
 userVariables [
  User.bIsFrontendPreview {type=NiagaraBool}
  User.bIsPainted {type=NiagaraBool}
  User.ColorVariantID {type=NiagaraInt32, defaultValue=4}
  User.DTCarOffset {type=Vector3f, description="", defaultValue {struct=Vector3f, value=<0 0 0>}}
  User.intensity_space {type=NiagaraFloat, defaultValue=0, description=""}
  User.intensity_trail {type=NiagaraFloat, defaultValue=3, description=""}
  User.Lines_ribbon {type=NiagaraFloat, defaultValue=5, description=""}
  User.LinesinStar {type=NiagaraFloat, defaultValue=1.5, description=""}
  User.OpacityMiniStars {type=NiagaraFloat, defaultValue=30, description=""}
  User.PaintedColor {type=LinearColor, description="", defaultValue {struct=LinearColor, value=<rgba 1 1 1 1>}}
  User.PaintedColor2 {type=LinearColor, description="", defaultValue {struct=LinearColor, value=<rgba 1 1 1 1>}}
  User.PaintedColor3 {type=LinearColor, description="", defaultValue {struct=LinearColor, value=<rgba 1 1 1 1>}}
  User.PaintedScalar {type=NiagaraFloat, defaultValue=1, description=""}
  User.PortCount {type=NiagaraInt32, defaultValue=1}
  User.SingleBoostPos {type=NiagaraPosition, description="", defaultValue {}}
 ]
}
