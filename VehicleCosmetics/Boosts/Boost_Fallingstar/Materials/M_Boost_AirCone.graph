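# Translucent unlit material for the air cone mesh particle of the Fallingstar vehicle boost.
# Panning air-effect textures, one UV-distorted by plasma-ribbon noise, combine with V-axis gradients and particle color into glowing spikes and lines through MF_FX_EmissiveAdjustments.
# An opacity mask reuses those layers, scaled by particle alpha.
# A WPO reroute builds parameterized per-axis sine drift, plus a V-weighted traveling wave on X and Z.
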Material /VehicleCosmetics/Boosts/Boost_Fallingstar/Materials/M_Boost_AirCone
graph {
 properties {
  assetImportData="<self>:InterchangeAssetImportData_0"
  bUsedWithNanite=true
  bUsedWithNiagaraMeshParticles=true
  blendMode=BLEND_Translucent
  shadingModel=MSM_Unlit
  usedShadingModels=Unlit
 }
 nodes [
  n1 = ComponentMask(None=TextureCoordinate()) {g=true}
  MaterialExpressionMeshEmitterVertexColor_2 = ParticleColor()
  n2 = TextureCoordinate()
  n3 = TextureCoordinate()
  n4 = Multiply(
   A=Power(Base=OneMinus(None=@n2), Exp=Constant() {r=1})
   B=TextureSample(UVs=Panner(Coordinate=@n3) {speedX=0.2, speedY=1.3}) {texture=/DealershipSharedLibrary/FX/Textures/TexturePackage/T_VCVFX_PlasmaRibbonMask02_RGB}.G
  )
  n5 = Constant() {r=0.02}
  n6 = TextureCoordinate()
  n7 = TextureSample(UVs=Add(A=Multiply(A=@n4, B=@n5), B=Panner(Coordinate=@n6))) {texture=/VehicleCosmetics/Boosts/Boost_Fallingstar/Textures/T_VCVFX_Boost_FallingStar_AirEffect_RGB}
  n8 = Power(Base=ComponentMask(None=TextureCoordinate()) {g=true}, Exp=Constant() {r=1.1})
  n9 = Constant() {r=2}
  n10 = VectorParameter() {parameterName=WhiteColorElements, defaultValue=<rgba 1 1 1 1>}
  n11 = Panner(Coordinate=TextureCoordinate()) {speedX=0.5, speedY=1}
  n12 = ComponentMask(None=TextureCoordinate()) {g=true}
  n13 = OneMinus(None=@n12)
  n14 = Multiply(
   A=Multiply(
    A=TextureSample(UVs=@n11) {texture=&T_VCVFX_Boost_FallingStar_AirEffect_RGB}.G
    B=ScalarParameter() {defaultValue=0.5, parameterName=OpacityLine}
   )
   B=Multiply(A=Multiply(A=@n12, B=@n13), B=Constant() {r=3})
  )
  n15 = Add(A=Multiply(A=@n7.R, B=Multiply(A=@n8, B=@n9)), B=Multiply(A=Add(A=@MaterialExpressionMeshEmitterVertexColor_2.RGB, B=@n10.RGB), B=@n14))
  MaterialExpressionConstantClamp_5 = Saturate(None=Max(A=Round(None=@n1), B=Multiply(A=@MaterialExpressionMeshEmitterVertexColor_2.A, B=@n15)))
  DynamicParameter() {parameterIndex=1, paramNames [EmissiveSpikesColor, GlowLineColor, HardGradientColor, Param4]}
  n16 = ComponentMask(None=TextureCoordinate()) {g=true}
  n17 = Power(Base=OneMinus(None=@n16), Exp=Constant() {r=0.05})
  n18 = Multiply(A=Power(Base=@n16, Exp=Constant() {r=5}), B=@n10.RGB)
  n19 = Add(
   A=Multiply(
    A=Multiply(A=@n17, B=@MaterialExpressionMeshEmitterVertexColor_2.RGB)
    B=ScalarParameter() {defaultValue=1, parameterName=EmissiveSpikesColor}
   )
   B=Multiply(A=@n18, B=ScalarParameter() {defaultValue=5, parameterName=HardGradientColor})
  )
  n20 = Add(A=ComponentMask(None=@n2) {g=true}, B=@n7.R)
  n21 = Multiply(A=Add(A=@MaterialExpressionMeshEmitterVertexColor_2.RGB, B=@n18), B=@n14)
  n22 = ScalarParameter() {defaultValue=1, parameterName=GlowLineColor}
  MaterialExpressionConstantClamp_4 = Saturate(None=Add(A=Multiply(A=@n19, B=@n20), B=Multiply(A=@n21, B=@n22)))
  MaterialFunctionCall(Emissive=@MaterialExpressionConstantClamp_4) {materialFunction=&MF_FX_EmissiveAdjustments}
  n23 = Constant() {r=1}
  n24 = Sine(None=Multiply(A=ScalarParameter() {defaultValue=1.5, parameterName=SpeedDB}, B=Time()))
  n25 = Sine(None=Multiply(A=ScalarParameter() {defaultValue=1, parameterName=SpeedAB}, B=Time()))
  n26 = OneMinus(None=ComponentMask(None=TextureCoordinate()) {g=true})
  n27 = Constant() {r=3}
  n28 = TextureCoordinate() {coordinateIndex=1}
  n29 = Time()
  n30 = Add(
   A=Divide(A=ComponentMask(None=@n28) {g=true}, B=ScalarParameter() {defaultValue=0.8, parameterName=CountWaves2})
   B=Multiply(A=ScalarParameter() {defaultValue=-0.5, parameterName=VerticalSpeed2}, B=OneMinus(None=@n29))
  )
  n31 = Multiply(A=Multiply(A=Power(Base=@n26, Exp=@n27), B=Constant() {r=3}), B=Multiply(A=Sine(None=@n30) {period=0.3}, B=Constant() {r=0.3}))
  n32 = Add(A=Multiply(A=LinearInterpolate(A=@n23, B=@n24, Alpha=@n25), B=ScalarParameter() {defaultValue=1, parameterName=X_Drift_Mult}), B=@n31)
  n33 = Multiply(A=ScalarParameter() {defaultValue=1.5, parameterName=SpeedD}, B=Time())
  n34 = Multiply(A=ScalarParameter() {defaultValue=1, parameterName=SpeedA}, B=Time())
  n35 = Multiply(
   A=LinearInterpolate(A=Constant() {r=1}, B=Sine(None=@n33), Alpha=Sine(None=@n34))
   B=ScalarParameter() {defaultValue=1, parameterName=Y_Drift_Mult}
  )
  n36 = Multiply(A=ScalarParameter() {defaultValue=1.5, parameterName=SpeedC}, B=Time())
  n37 = Multiply(A=ScalarParameter() {defaultValue=1, parameterName=SpeedB}, B=Add(A=Time(), B=ScalarParameter() {defaultValue=1, parameterName=TimeOffset}))
  n38 = Multiply(
   A=LinearInterpolate(A=Sine(None=@n36), B=Constant() {r=-1}, Alpha=Cosine(None=@n37))
   B=ScalarParameter() {defaultValue=1, parameterName=Z_Drift_Mult}
  )
  n39 = Multiply(
   A=AppendVector(A=AppendVector(A=@n32, B=@n35), B=Add(A=@n38, B=@n31))
   B=ScalarParameter() {defaultValue=5, parameterName=Uniform_Drift_Mult}
  )
  n40 = VectorParameter() {parameterName=WPO_OffsetB, defaultValue=<rgba 0 0 0 1>}
  NamedRerouteDeclaration(None=Add(A=ComponentMask(None=@n39) {b=true, g=true, r=true}, B=ComponentMask(None=@n40.RGB) {b=true, g=true, r=true})) {name=WPO, nodeColor=<rgba 1 0.0022 0.6514 1>}
  NamedRerouteUsage()
 ]
}
