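# Unlit masked material for the Fallingstar vehicle boost's air cone heat effect, usable on Niagara mesh particles.
# Emissive mixes particle-tinted and white vertical UV gradients with panning air-effect textures, one distorted by plasma noise, through MF_FX_EmissiveAdjustments.
# A mask combines those textures with particle alpha and a rounded V gradient.
# A WPO reroute builds time-based sine drift plus gradient-weighted waves from a second UV channel.
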Material /VehicleCosmetics/Boosts/Boost_Fallingstar/Materials/M_Boost_AirConeHeatMask
graph {
 properties {
  assetImportData="<self>:InterchangeAssetImportData_0"
  bUsedWithNanite=true
  bUsedWithNiagaraMeshParticles=true
  blendMode=BLEND_Masked
  previewMesh=/VehicleCosmetics/Boosts/Boost_Fallingstar/Meshes/FallingStar_SM_V3
  shadingModel=MSM_Unlit
  usedShadingModels=Unlit
 }
 nodes [
  n1 = ComponentMask(None=TextureCoordinate()) {g=true}
  MaterialExpressionMeshEmitterVertexColor_3 = ParticleColor()
  n2 = Multiply(A=Power(Base=OneMinus(None=@n1), Exp=Constant() {r=0.05}), B=@MaterialExpressionMeshEmitterVertexColor_3.RGB)
  n3 = ScalarParameter() {defaultValue=1, parameterName=EmissiveSpikesColor}
  n4 = Multiply(A=Power(Base=@n1, Exp=Constant() {r=5}), B=VectorParameter() {parameterName=WhiteColorElements, defaultValue=<rgba 1 1 1 1>}.RGB)
  n5 = ScalarParameter() {defaultValue=5, parameterName=HardGradientColor}
  n6 = ComponentMask(None=TextureCoordinate()) {g=true}
  n7 = TextureCoordinate()
  n8 = Multiply(
   A=Power(Base=OneMinus(None=@n6), Exp=Constant() {r=1})
   B=TextureSample(UVs=Panner(Coordinate=@n7) {speedX=0.2, speedY=1.3}) {texture=/DealershipSharedLibrary/FX/Textures/TexturePackage/T_VCVFX_PlasmaRibbonMask02_RGB}.G
  )
  n9 = Constant() {r=0.02}
  n10 = TextureCoordinate()
  n11 = TextureSample(UVs=Add(A=Multiply(A=@n8, B=@n9), B=Panner(Coordinate=@n10))) {texture=/VehicleCosmetics/Boosts/Boost_Fallingstar/Textures/T_VCVFX_Boost_FallingStar_AirEffect_RGB}
  n12 = Multiply(A=Add(A=Multiply(A=@n2, B=@n3), B=Multiply(A=@n4, B=@n5)), B=Add(A=@n6, B=@n11.R))
  n13 = TextureSample(UVs=Panner(Coordinate=TextureCoordinate()) {speedX=0.5, speedY=1}) {texture=&T_VCVFX_Boost_FallingStar_AirEffect_RGB}
  n14 = ScalarParameter() {defaultValue=0.5, parameterName=OpacityLine}
  n15 = ComponentMask(None=TextureCoordinate()) {g=true}
  n16 = Multiply(A=@n15, B=OneMinus(None=@n15))
  n17 = Constant() {r=3}
  n18 = Multiply(A=Add(A=@MaterialExpressionMeshEmitterVertexColor_3.RGB, B=@n4), B=Multiply(A=Multiply(A=@n13.G, B=@n14), B=Multiply(A=@n16, B=@n17)))
  MaterialFunctionCall(Emissive=Saturate(None=Add(A=@n12, B=@n18))) {materialFunction=&MF_FX_EmissiveAdjustments}
  n19 = Constant() {r=1}
  n20 = Sine(None=Multiply(A=ScalarParameter() {defaultValue=1.5, parameterName=SpeedDB}, B=Time()))
  n21 = Sine(None=Multiply(A=ScalarParameter() {defaultValue=1, parameterName=SpeedAB}, B=Time()))
  n22 = OneMinus(None=ComponentMask(None=TextureCoordinate()) {g=true})
  n23 = Constant() {r=3}
  n24 = TextureCoordinate() {coordinateIndex=1}
  n25 = Time()
  n26 = Add(
   A=Divide(A=ComponentMask(None=@n24) {g=true}, B=ScalarParameter() {defaultValue=0.8, parameterName=CountWaves2})
   B=Multiply(A=ScalarParameter() {defaultValue=-0.5, parameterName=VerticalSpeed2}, B=OneMinus(None=@n25))
  )
  n27 = Multiply(A=Multiply(A=Power(Base=@n22, Exp=@n23), B=Constant() {r=3}), B=Multiply(A=Sine(None=@n26) {period=0.3}, B=Constant() {r=0.3}))
  n28 = Add(A=Multiply(A=LinearInterpolate(A=@n19, B=@n20, Alpha=@n21), B=ScalarParameter() {defaultValue=1, parameterName=X_Drift_Mult}), B=@n27)
  n29 = Multiply(A=ScalarParameter() {defaultValue=1.5, parameterName=SpeedD}, B=Time())
  n30 = Multiply(A=ScalarParameter() {defaultValue=1, parameterName=SpeedA}, B=Time())
  n31 = Multiply(
   A=LinearInterpolate(A=Constant() {r=1}, B=Sine(None=@n29), Alpha=Sine(None=@n30))
   B=ScalarParameter() {defaultValue=1, parameterName=Y_Drift_Mult}
  )
  n32 = Multiply(A=ScalarParameter() {defaultValue=1.5, parameterName=SpeedC}, B=Time())
  n33 = Multiply(A=ScalarParameter() {defaultValue=1, parameterName=SpeedB}, B=Add(A=Time(), B=ScalarParameter() {defaultValue=1, parameterName=TimeOffset}))
  n34 = Multiply(
   A=LinearInterpolate(A=Sine(None=@n32), B=Constant() {r=-1}, Alpha=Cosine(None=@n33))
   B=ScalarParameter() {defaultValue=1, parameterName=Z_Drift_Mult}
  )
  n35 = Multiply(
   A=AppendVector(A=AppendVector(A=@n28, B=@n31), B=Add(A=@n34, B=@n27))
   B=ScalarParameter() {defaultValue=5, parameterName=Uniform_Drift_Mult}
  )
  n36 = VectorParameter() {parameterName=WPO_OffsetB, defaultValue=<rgba 0 0 0 1>}
  NamedRerouteDeclaration(None=Add(A=ComponentMask(None=@n35) {b=true, g=true, r=true}, B=ComponentMask(None=@n36.RGB) {b=true, g=true, r=true})) {name=WPO, nodeColor=<rgba 0 0.4844 1 1>}
  NamedRerouteUsage()
  n37 = ComponentMask(None=TextureCoordinate()) {g=true}
  n38 = Power(Base=ComponentMask(None=TextureCoordinate()) {g=true}, Exp=Constant() {r=1.1})
  n39 = Constant() {r=2}
  n40 = Add(A=Multiply(A=@n11.R, B=Multiply(A=@n38, B=@n39)), B=@n18)
  Saturate(None=Max(A=Round(None=@n37), B=Multiply(A=@MaterialExpressionMeshEmitterVertexColor_3.A, B=@n40)))
 ]
}
