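# Unlit masked material for the Falling Star vehicle boost's star, usable on Niagara mesh particles and Nanite.
# Inner and outer Fresnel glows multiply particle color, a mini-star texture and a vertical position-based gradient into emissive.
# The texture scrolls and shifts with tangent-space view direction for fake depth; dynamic parameters control its strength and blend the gradient toward particle-colored paint.
# Nested time-driven sine and cosine blends build a per-axis drift offset.
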Material /VehicleCosmetics/Boosts/Boost_Fallingstar/Materials/M_Boost_Star
graph {
 properties {
  assetImportData="<self>:InterchangeAssetImportData_0"
  bUsedWithNanite=true
  bUsedWithNiagaraMeshParticles=true
  blendMode=BLEND_Masked
  shadingModel=MSM_Unlit
  usedShadingModels=Unlit
 }
 nodes [
  n1 = ScalarParameter() {defaultValue=1.5, parameterName=SpeedDB}
  n2 = Time()
  n3 = ScalarParameter() {defaultValue=1, parameterName=SpeedAB}
  n4 = Time()
  n5 = LinearInterpolate(A=Constant() {r=1}, B=Sine(None=Multiply(A=@n1, B=@n2)), Alpha=Sine(None=Multiply(A=@n3, B=@n4)))
  n6 = ScalarParameter() {defaultValue=1, parameterName=X_Drift_Mult}
  n7 = ScalarParameter() {defaultValue=1.5, parameterName=SpeedD}
  n8 = Time()
  n9 = ScalarParameter() {defaultValue=1, parameterName=SpeedA}
  n10 = Time()
  n11 = LinearInterpolate(A=Constant() {r=1}, B=Sine(None=Multiply(A=@n7, B=@n8)), Alpha=Sine(None=Multiply(A=@n9, B=@n10)))
  n12 = ScalarParameter() {defaultValue=1, parameterName=Y_Drift_Mult}
  n13 = Sine(None=Multiply(A=ScalarParameter() {defaultValue=1.5, parameterName=SpeedC}, B=Time()))
  n14 = Constant() {r=-1}
  n15 = Time()
  n16 = ScalarParameter() {defaultValue=1, parameterName=TimeOffset}
  n17 = Cosine(None=Multiply(A=ScalarParameter() {defaultValue=1, parameterName=SpeedB}, B=Add(A=@n15, B=@n16)))
  n18 = AppendVector(
   A=AppendVector(A=Multiply(A=@n5, B=@n6), B=Multiply(A=@n11, B=@n12))
   B=Multiply(A=LinearInterpolate(A=@n13, B=@n14, Alpha=@n17), B=ScalarParameter() {defaultValue=1, parameterName=Z_Drift_Mult})
  )
  n19 = ScalarParameter() {defaultValue=5, parameterName=Uniform_Drift_Mult}
  Add(
   A=ComponentMask(None=Multiply(A=@n18, B=@n19)) {b=true, g=true, r=true}
   B=ComponentMask(None=VectorParameter() {parameterName=WPO_OffsetB, defaultValue=<rgba 0 0 0 1>}.RGB) {b=true, g=true, r=true}
  )
  n20 = Add(
   A=MaterialFunctionCall(Invert Fresnel=StaticBool() {value=true}, Power=ScalarParameter() {defaultValue=4, parameterName=FresInsideGlow}) {materialFunction=&Fresnel_Function}.Result
   B=MaterialFunctionCall(Invert Fresnel=StaticBool(), Power=ScalarParameter() {defaultValue=5, parameterName=FresOutsideGlow}) {materialFunction=&Fresnel_Function}.Result
  )
  n21 = ScalarParameter() {defaultValue=10, parameterName=PowerFresnel}
  MaterialExpressionVertexColor_0 = ParticleColor()
  n22 = Multiply(A=@MaterialExpressionVertexColor_0.RGB, B=VectorParameter() {parameterName=Space Color Mult, defaultValue=<rgba 5 5 5 1>}.RGB)
  MaterialExpressionMeshEmitterDynamicParameter_0 = DynamicParameter() {defaultValue=<rgba 1 1.5 1.5 1>, paramNames [Painted?, OpacityMiniStars, LinesinStar, Param4]}
  n23 = Constant() {r=1.1}
  n24 = Constant() {r=0.04}
  n25 = Constant() {r=1.1}
  n26 = AppendVector(A=Multiply(A=Time(), B=Multiply(A=@n23, B=@n24)), B=Multiply(A=Time(), B=Multiply(A=@n24, B=@n25)))
  n27 = CameraVectorWS()
  n28 = Multiply(
   A=Constant2Vector() {g=30, r=12}
   B=ComponentMask(None=Transform(None=@n27) {transformSourceType=TRANSFORMSOURCE_World, transformType=TRANSFORM_Tangent}) {g=true, r=true}
  )
  n29 = Multiply(
   A=@MaterialExpressionMeshEmitterDynamicParameter_0.LinesinStar
   B=TextureSample(UVs=Add(A=@n26, B=@n28)) {texture=/VehicleCosmetics/Boosts/Boost_Fallingstar/Textures/T_VCVFX_Boost_FallingStar_Star_RGB}.R
  )
  MaterialExpressionMeshEmitterVertexColor_0 = ParticleColor()
  n30 = Add(
   A=Multiply(A=Multiply(A=@n22, B=@n29), B=@MaterialExpressionMeshEmitterDynamicParameter_0.OpacityMiniStars)
   B=@MaterialExpressionMeshEmitterVertexColor_0.RGB
  )
  n31 = WorldPosition()
  n32 = ScalarParameter() {defaultValue=12, parameterName=DivideWP}
  MaterialExpressionObjectWorldPosition_0 = ObjectPositionWS()
  n33 = Subtract(
   A=Add(A=Divide(A=@n31.XYZ, B=@n32), B=ScalarParameter() {defaultValue=0.5, parameterName=AddValue})
   B=Add(A=@MaterialExpressionObjectWorldPosition_0, B=ScalarParameter() {parameterName=ObjectWP})
  )
  MaterialExpressionConstantClamp_0 = Saturate(None=OneMinus(None=ComponentMask(None=@n33) {b=true}))
  n34 = ScalarParameter() {defaultValue=2, parameterName=PowerWP}
  n35 = Multiply(
   A=Constant3Vector() {constant=<rgba 0.5 0.1 0.6 0>}
   B=Multiply(A=Power(Base=@MaterialExpressionConstantClamp_0, Exp=@n34), B=ScalarParameter() {defaultValue=1, parameterName=Brightness})
  )
  n36 = Desaturation(None=@n35)
  n37 = Constant() {r=10}
  n38 = LinearInterpolate(
   A=@n35
   B=Multiply(A=Multiply(A=@n36, B=@n37), B=@MaterialExpressionMeshEmitterVertexColor_0.RGB)
   Alpha=@MaterialExpressionMeshEmitterDynamicParameter_0.Painted?
  )
  MaterialFunctionCall(Emissive=Multiply(A=Multiply(A=@n20, B=@n21), B=Add(A=@n30, B=@n38))) {materialFunction=&MF_FX_EmissiveAdjustments}
 ]
}
