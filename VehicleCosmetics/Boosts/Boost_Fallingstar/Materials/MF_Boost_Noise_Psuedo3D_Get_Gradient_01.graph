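# Material function that returns a pseudo-random 2D gradient vector for pseudo-3D noise in the Fallingstar vehicle boost materials.
# It hashes the integer position input with the dot, sine, frac trick to get a random value per position.
# That value sets a starting angle and a rotation rate scaled by the t input, so each gradient spins at its own speed as t changes.
# Cosine and sine of the angle form the output unit vector.
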MaterialFunction /VehicleCosmetics/Boosts/Boost_Fallingstar/Materials/MF_Boost_Noise_Psuedo3D_Get_Gradient_01
graph {nodes [
  n1 = DotProduct(
   A=FunctionInput(Preview=TextureCoordinate()) {bUsePreviewValueAsDefault=true, inputName="intPos (uv)", inputType=FunctionInput_Vector2}
   B=Constant2Vector() {g=78.233, r=12.9898}
  )
  n2 = Frac(None=Multiply(A=Sine(None=@n1), B=Constant() {r=43758.543}))
  n3 = Constant() {r=4}
  n4 = FunctionInput() {bUsePreviewValueAsDefault=true, desc=t, inputName=t, inputType=FunctionInput_Scalar, sortPriority=1, previewValue=<1 0 0 1>}
  n5 = Add(A=Multiply(A=@n2, B=Constant() {r=6.2832}), B=Multiply(A=Multiply(A=@n2, B=@n3), B=@n4))
  FunctionOutput(None=AppendVector(A=Cosine(None=@n5), B=Sine(None=@n5)))
 ]}
