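# Material function for the Fallingstar vehicle boost that generates time-animated, pseudo-3D gradient noise from 2D UVs.
# UVs are multiplied by Scale, aspect ratio and a constant, then split into integer cell and fractional position.
# A helper function supplies time-varying gradients at the four cell corners, whose dot products blend through smoothstep-weighted lerps, Perlin-style.
# Outputs are remapped noise as float and float3, plus an absolute-value plasma variant.
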MaterialFunction /VehicleCosmetics/Boosts/Boost_Fallingstar/Materials/MF_Boost_Noise_Psuedo3D_01
graph {nodes [
  n1 = FunctionInput() {bUsePreviewValueAsDefault=true, inputName=Scale, inputType=FunctionInput_Scalar, sortPriority=1, previewValue=<1 0 0 1>}
  n2 = Multiply(A=FunctionInput(Preview=TextureCoordinate()) {bUsePreviewValueAsDefault=true, inputName=UV, inputType=FunctionInput_Vector2}, B=@n1)
  n3 = FunctionInput(Preview=Constant2Vector() {g=1, r=1}) {bUsePreviewValueAsDefault=true, inputName=aspectRatio, inputType=FunctionInput_Vector2, sortPriority=2}
  n4 = AppendVector(
   A=Multiply(A=Multiply(A=@n2, B=@n3), B=Constant() {r=10})
   B=FunctionInput(Preview=Time()) {bUsePreviewValueAsDefault=true, inputName=Time, inputType=FunctionInput_Scalar, sortPriority=3}
  )
  n5 = ComponentMask(None=@n4) {g=true, r=true}
  n6 = Floor(None=@n5)
  n7 = Constant2Vector()
  n8 = Add(A=@n6, B=@n7)
  n9 = ComponentMask(None=@n4) {b=true}
  n10 = Subtract(A=@n5, B=@n6)
  n11 = DotProduct(
   A=MaterialFunctionCall!("intPos (uv)"=AppendVector(A=@n8, B=@n9), t=@n9) {materialFunction=&/MF_Boost_Noise_Psuedo3D_Get_Gradient_01}.Result
   B=Subtract(A=@n10, B=@n7)
  )
  n12 = Constant2Vector() {r=1}
  n13 = Add(A=@n6, B=@n12)
  n14 = DotProduct(
   A=MaterialFunctionCall("intPos (uv)"=AppendVector(A=@n13, B=@n9), t=@n9) {materialFunction=&MF_Boost_Noise_Psuedo3D_Get_Gradient_01}.Result
   B=Subtract(A=@n10, B=@n12)
  )
  n15 = Constant() {r=3}
  n16 = Multiply(A=@n10, B=Constant() {r=2})
  n17 = Multiply(A=Multiply(A=Subtract(A=@n15, B=@n16), B=@n10), B=@n10)
  n18 = ComponentMask(None=@n17) {r=true}
  n19 = Constant2Vector() {g=1}
  n20 = DotProduct(
   A=MaterialFunctionCall("intPos (uv)"=Add(A=@n6, B=@n19), t=@n9) {materialFunction=&MF_Boost_Noise_Psuedo3D_Get_Gradient_01}.Result
   B=Subtract(A=@n10, B=@n19)
  )
  n21 = Constant2Vector() {g=1, r=1}
  n22 = DotProduct(
   A=MaterialFunctionCall("intPos (uv)"=Add(A=@n6, B=@n21), t=@n9) {materialFunction=&MF_Boost_Noise_Psuedo3D_Get_Gradient_01}.Result
   B=Subtract(A=@n10, B=@n21)
  )
  n23 = Divide(
   A=LinearInterpolate(A=LinearInterpolate(A=@n11, B=@n14, Alpha=@n18), B=LinearInterpolate(A=@n20, B=@n22, Alpha=@n18), Alpha=ComponentMask(None=@n17) {g=true})
   B=Constant() {r=0.7}
  )
  n24 = Constant() {r=0.5}
  n25 = Add(A=Multiply(A=@n23, B=@n24), B=@n24)
  FunctionOutput(None=AppendVector(A=AppendVector(A=@n25, B=@n25), B=@n25)) {outputName=Float3 Noise}
  FunctionOutput(None=@n25) {outputName=Float Noise}
  FunctionOutput(None=Abs(None=@n23)) {outputName=Float Plasma1}
 ]}
