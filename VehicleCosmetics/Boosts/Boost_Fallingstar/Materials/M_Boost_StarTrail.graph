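# Translucent unlit Niagara ribbon material for the star trail of the Falling Star vehicle boost.
# Camera-vector-scrolled star texture, boosted by dynamic-parameter intensity, combines with air-effect and star-line textures, tinted by particle color for emissive.
# Opacity multiplies that pattern by particle alpha, depth fade and front/back gradient fades.
# A world-position-offset chain sums time-driven sine drift, a vector offset and pseudo-3D noise.
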Material /VehicleCosmetics/Boosts/Boost_Fallingstar/Materials/M_Boost_StarTrail
graph {
 properties {
  bAutomaticallySetUsageInEditor=false
  bUsedWithNiagaraRibbons=true
  blendMode=BLEND_Translucent
  shadingModel=MSM_Unlit
  usedShadingModels=Unlit
 }
 nodes [
  n1 = Multiply(A=ScalarParameter() {defaultValue=1.5, parameterName=SpeedDB}, B=Time())
  n2 = Multiply(A=ScalarParameter() {defaultValue=1, parameterName=SpeedAB}, B=Time())
  n3 = Multiply(
   A=LinearInterpolate(A=Constant() {r=1}, B=Sine(None=@n1), Alpha=Sine(None=@n2))
   B=ScalarParameter() {defaultValue=1, parameterName=X_Drift_Mult}
  )
  n4 = Multiply(A=ScalarParameter() {defaultValue=1.5, parameterName=SpeedD}, B=Time())
  n5 = Multiply(A=ScalarParameter() {defaultValue=1, parameterName=SpeedA}, B=Time())
  n6 = Multiply(
   A=LinearInterpolate(A=Constant() {r=1}, B=Sine(None=@n4), Alpha=Sine(None=@n5))
   B=ScalarParameter() {defaultValue=1, parameterName=Y_Drift_Mult}
  )
  n7 = ScalarParameter() {defaultValue=1.5, parameterName=SpeedC}
  n8 = Time()
  n9 = ScalarParameter() {defaultValue=1, parameterName=SpeedB}
  n10 = Add(A=Time(), B=ScalarParameter() {defaultValue=1, parameterName=TimeOffset})
  n11 = LinearInterpolate(A=Sine(None=Multiply(A=@n7, B=@n8)), B=Constant() {r=-1}, Alpha=Cosine(None=Multiply(A=@n9, B=@n10)))
  n12 = ScalarParameter() {defaultValue=1, parameterName=Z_Drift_Mult}
  n13 = Multiply(
   A=AppendVector(A=AppendVector(A=@n3, B=@n6), B=Multiply(A=@n11, B=@n12))
   B=ScalarParameter() {defaultValue=5, parameterName=Uniform_Drift_Mult}
  )
  n14 = VectorParameter() {parameterName=WPO_OffsetB, defaultValue=<rgba 0 0 0 1>}
  n15 = TextureCoordinate()
  n16 = ScalarParameter() {defaultValue=2, group=Noise, parameterName=NoiseScale}
  n17 = ComponentMask(None=VectorParameter() {group=Noise, parameterName=NoiseAspectRation, defaultValue=<rgba 0.5 1 0 1>}.RGB) {b=true, g=true, r=true}
  n18 = Multiply(A=ScalarParameter() {defaultValue=0.7, group=Noise, parameterName=NoiseTime}, B=Time())
  Add(
   A=Add(A=ComponentMask(None=@n13) {b=true, g=true, r=true}, B=ComponentMask(None=@n14.RGB) {b=true, g=true, r=true})
   B=Multiply(B=MaterialFunctionCall!(UV=@n15, Scale=@n16, aspectRatio=@n17, Time=@n18) {materialFunction=&/MF_Boost_Noise_Psuedo3D_01}.Float Noise) {constA=5}
  )
  MaterialExpressionVertexColor_1 = ParticleColor()
  n19 = DynamicParameter() {defaultValue=<rgba 1 1 5 1>, paramNames [is painteed, intensity, Intensity_lines_ribbon, Param4]}
  n20 = Power(Base=@n19.intensity, Exp=@n19.intensity)
  n21 = Constant() {r=0.5}
  n22 = Constant() {r=0.1}
  n23 = Constant() {r=-1.1}
  n24 = AppendVector(A=Multiply(A=Time(), B=Multiply(A=@n21, B=@n22)), B=Multiply(A=Time(), B=Multiply(A=@n22, B=@n23)))
  n25 = CameraVectorWS()
  n26 = Multiply(A=Constant2Vector() {g=4, r=4}, B=ComponentMask(None=Transform(None=@n25) {transformType=TRANSFORM_Tangent}) {b=true, r=true})
  n27 = Multiply(
   A=Constant() {r=6}
   B=TextureSample(UVs=Add(A=@n24, B=@n26) {constA=1}) {texture=/VehicleCosmetics/Boosts/Boost_Fallingstar/Textures/T_VCVFX_Boost_FallingStar_Star_RGB}.R
  )
  n28 = Constant() {r=0.7}
  n29 = TextureSample() {texture=/VehicleCosmetics/Boosts/Boost_Fallingstar/Textures/T_VCVFX_Boost_FallingStar_AirEffect_RGB}
  n30 = TextureSample() {texture=&T_VCVFX_Boost_FallingStar_Star_RGB}
  n31 = Add(
   A=Multiply(A=Multiply(A=@n20, B=@n27), B=Constant() {r=1.5})
   B=Add(A=Multiply(A=@n28, B=@n29.B), B=Multiply(A=@n30.B, B=@n19.Intensity_lines_ribbon))
  )
  MaterialFunctionCall(Emissive=Saturate(None=Multiply(A=@MaterialExpressionVertexColor_1.RGB, B=@n31))) {materialFunction=&MF_FX_EmissiveAdjustments}
  n32 = Multiply(A=@MaterialExpressionVertexColor_1.A, B=@n31)
  n33 = Constant() {r=50}
  n34 = ScalarParameter() {defaultValue=1, group=Opacity, parameterName=Contrast Fade Front, sortPriority=5}
  n35 = ScalarParameter() {defaultValue=1, group=Opacity, parameterName=Brightness Fade Front, sortPriority=6}
  n36 = Constant() {r=1}
  n37 = ScalarParameter() {defaultValue=1, group=Opacity, parameterName=Contrast Fade Back, sortPriority=7}
  n38 = ScalarParameter() {defaultValue=1, group=Opacity, parameterName=Brightness Fade Back, sortPriority=8}
  Multiply(
   A=Saturate(None=MaterialFunctionCall(Opacity In=@n32, DepthFade Distance=@n33) {materialFunction=&MF_VCVFX_DepthFade}.Result)
   B=Multiply(
    A=MaterialFunctionCall!(
     "Gradient U=0 V=1"
     "Invert gradient 0=False 1=True"
     "Contrast (Power)"=@n34
     "Brightness (Multiply)"=@n35
     "Dynamic Contrast (Power)"
     "Dynamic Brightness (Multiply)"
    ) {materialFunction=/DealershipSharedLibrary/FX/Materials/MaterialFunctions/MF_LinearGradientPlus}.Result
    B=MaterialFunctionCall("Invert gradient 0=False 1=True"=@n36, "Contrast (Power)"=@n37, "Brightness (Multiply)"=@n38) {materialFunction=&MF_LinearGradientPlus}.Result
   )
  )
  n39 = TextureCoordinate() {uTiling=3}
  Multiply(B=TextureSample(UVs=Panner(Coordinate=@n39) {speedX=-1.5}) {texture=/DealershipSharedLibrary/FX/Textures/Gradient_Masks/T_VCVFX_BoostKittyRareMask_A}.R)
  Transform(None=@n25) {transformType=TRANSFORM_Tangent}
  VectorParameter() {parameterName=Space Color Mult, defaultValue=<rgba 5 5 5 1>}
 ]
}
